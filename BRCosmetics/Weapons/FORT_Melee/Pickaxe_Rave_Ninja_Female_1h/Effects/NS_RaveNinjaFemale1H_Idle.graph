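# Idle Niagara effect for the one-handed female Rave Ninja pickaxe, adding neon sprites, a soft glow and a light.
# DecalsStart repeatedly bursts cyan or pink sprites that follow three AnimTrail skeletal mesh sockets by sequential index, with curve-driven alpha and material parameters.
# LowGlow and LowGlow_LockX draw faint, offset cyan sprites locked to different axes, while Particle_Emitter supplies a cyan light.
# Several modules come from the Cascade-to-Niagara converter.
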NiagaraSystem /BRCosmetics/Weapons/FORT_Melee/Pickaxe_Rave_Ninja_Female_1h/Effects/NS_RaveNinjaFemale1H_Idle
dependencies {
 usedDataInterfaces [
  &NiagaraDataInterfaceArrayColor
  &NiagaraDataInterfaceCurve
  &NiagaraDataInterfaceSkeletalMesh
  &NiagaraDataInterfaceArrayFloat3
 ]
 usedDynamicInputs []
 usedModules [&SystemState]
}
emitters {
 DecalsStart {
  emitterUpdateScript [
   &EmitterState(Life Cycle Mode=::NewEnumerator1, Loop Duration=0.5, UseLoopDelay=true)
   &SpawnBurst_Instantaneous(Spawn Count=6, Use Spawn Probability=false, UseLoopCountLimit=false)
  ]
  particleSpawnScript [
   /SetVariables(Particles.SkeletonSamplingIndex:NiagaraInt32=/CascadeToNiagaraConverter/NiagaraScripts/DynamicInputScripts/SequentialInt(Use Array Data:NiagaraBool=false, First Number:NiagaraInt32=0, Sequence:NiagaraDataInterfaceArrayInt32=?, Last Number:NiagaraInt32=2)) {isSetParameters=true}
   &InitializeParticle(Sprite Size=<11 28>, Sprite Size Mode=::NewEnumerator4, UsePositionOffset=false, Write Material Random=true)
   &Color(Color=&MakeLinearColorFromVectorAndFloat(
     "Vector (RGB)"=&RandomRangeVector(Minimum=<0 32 26.7595>, Maximum=<32 7.3333 28.7625>, Evaluation Type=::NewEnumerator1)
     "Float (Alpha)"=0.6
    ))
   &Color(Color=&SelectLinearColorFromArray(
     Array Sampling Mode=::NewEnumerator0
     Direct Array Index=?
     Color Selection Array=DataInterface() {propertyValues {ColorData [<rgba 0 32 26.7595 1>, <rgba 32 7.3333 28.7625 1>]}}
     Random Seed=0
    )) {module=Color001}
   &CameraOffset(Camera Offset Amount=&RandomRangeFloat(Minimum=2, Maximum=20, Evaluation Type=::NewEnumerator1))
   &DynamicMaterialParameters(
    Index 0 Param 4=&RandomRangeFloat(Minimum=0, Maximum=1, Evaluation Type=::NewEnumerator1)
    Module.Param0WriteEnabled=false
    Module.Param1WriteEnabled=false
    Module.Param2WriteEnabled=false
    Module.Param3WriteEnabled=true
   ) {module=DynamicMaterialParameters001}
   /CascadeToNiagaraConverter/NiagaraScripts/ModuleScripts/SpriteLockAxis(Lock Axis Type:ENiagara_SpriteLockAxis=::NewEnumerator8, UseAlignmentCoordinateSpace=false, UseFacingCoordinateSpace=false)
  ]
  particleUpdateScript [
   &ParticleState()
   &ScaleColor(
    Scale Alpha=&FloatFromCurve(FloatCurve=DataInterface() {propertyValues {Curve {keys [<at 0 0 RCIM_Cubic>,<at 0.4502 1.736 RCIM_Cubic>,<at 1 0>]}}})
    ScaleA=true
    ScaleRGB=true
   )
   &SkeletalMeshLocation(
    Bone / Socket Index=$Particles.SkeletonSamplingIndex
    MaskSpawnGroup=false
    Mesh Sampling Type=::NewEnumerator5
    Orientation Sampling=::NewEnumerator1
    Sampled Position Offset=&SelectVectorFromArray(
     Direct Array Index=$Particles.SkeletonSamplingIndex
     Vector Selection Array=DataInterface() {propertyValues {FloatData [<0 -2.5 -13>, <-2.5 0 1>, <0 -2.5 10>]}}
    )
    Skeletal Mesh=DataInterface() {propertyValues {
      DefaultMesh=null
      ExcludeBoneName=root
      OptionalComponentClass=null
      PreviewMesh=null
      SkinningMode=SkinOnTheFly
      SoftSourceActor=null
      SourceMode=Default
      UvSetIndex=0
      WholeMeshLOD=-1
      bExcludeBone=false
      bReadDeformedGeometry=true
      bRequireCurrentFrameData=true
      ComponentTags []
      FilteredBones []
      FilteredSockets [AnimTrail_01,AnimTrail_03,AnimTrail_02]
      MeshUserParameter {parameter {name=None, typeDefHandle {registeredTypeIndex=1}, varData []}}
      SamplingRegions []
     }}
    Socket Sampling Mode=::NewEnumerator5
   )
   &DynamicMaterialParameters(
    Index 0 Param 1=2
    Index 0 Param 2=1
    Index 0 Param 3=&FloatFromCurve(FloatCurve=DataInterface() {propertyValues {Curve {keys [
         {arriveTangent=-0.9893, interpMode=RCIM_Cubic, leaveTangent=-0.9893, tangentMode=RCTM_User, time=0, value=-0.25}
         {arriveTangent=1.1441, interpMode=RCIM_Cubic, leaveTangent=1.1441, tangentMode=RCTM_User, time=0.4089, value=-0.0831}
         {arriveTangent=0.1844, interpMode=RCIM_Cubic, leaveTangent=0.1844, tangentMode=RCTM_User, time=1, value=0.5}
        ]}}})
    Module.Param0WriteEnabled=true
    Module.Param1WriteEnabled=true
    Module.Param2WriteEnabled=true
    Module.Param3WriteEnabled=false
   )
  ]
  rendererClasses [&NiagaraSpriteRendererProperties]
  renderers [{rendererClass=&NiagaraSpriteRendererProperties}]
 }
 LowGlow {
  emitterUpdateScript [
   &EmitterState(
    Inactive Response=::NewEnumerator1
    Life Cycle Mode=::NewEnumerator1
    Loop Behavior=::NewEnumerator1
    Loop Duration Mode=::NewEnumerator1
    UseLoopDelay=true
   )
   &SpawnBurst_Instantaneous(Use Spawn Probability=false, UseLoopCountLimit=false)
  ]
  particleSpawnScript [
   &InitializeParticle(Sprite Size=<25 85>, Sprite Size Mode=::NewEnumerator4, UsePositionOffset=false, Write Material Random=true)
   &Color(Color=&MakeLinearColorFromVectorAndFloat("Vector (RGB)"=<0 1 0.9752>, "Float (Alpha)"=0.2))
   /CascadeToNiagaraConverter/NiagaraScripts/ModuleScripts/AddVectorToPosition(Vector:Vector3f=&TransformVector(Vector=<0 0 30>, Destination Space=::Simulation))
   &SpriteLockAxis(Lock Axis Type:ENiagara_SpriteLockAxis=::NewEnumerator4, UseAlignmentCoordinateSpace=false, UseFacingCoordinateSpace=false)
  ]
  particleUpdateScript [
   &ParticleState(
    Kill Particles When Lifetime Has Elapsed=false
    Let Infinitely Lived Particles Die When Emitter Deactivates=true
    Loop Particles Lifetime=true
   )
   &DynamicMaterialParameters(
    Index 0 Param 2=1
    Index 0 Param 3=1
    Index 0 Param 4=1
    Module.Param0WriteEnabled=true
    Module.Param1WriteEnabled=true
    Module.Param2WriteEnabled=true
    Module.Param3WriteEnabled=true
   )
  ]
  rendererClasses [&NiagaraSpriteRendererProperties]
  renderers [{rendererClass=&NiagaraSpriteRendererProperties}]
 }
 LowGlow_LockX {
  emitterUpdateScript [
   &EmitterState()
   &SpawnBurst_Instantaneous(Use Spawn Probability=false, UseLoopCountLimit=true)
  ]
  particleSpawnScript [
   &InitializeParticle(Sprite Size=<10 85>, Sprite Size Mode=::NewEnumerator4, UsePositionOffset=false, Write Material Random=true)
   &Color(Color=&MakeLinearColorFromVectorAndFloat("Vector (RGB)"=<0 1 0.9752>, "Float (Alpha)"=0.2))
   &AddVectorToPosition(Vector:Vector3f=&TransformVector(Vector=<0 0 30>, Destination Space=::Simulation))
   &SpriteLockAxis(Lock Axis Type:ENiagara_SpriteLockAxis=::NewEnumerator0, UseAlignmentCoordinateSpace=false, UseFacingCoordinateSpace=false)
  ]
  particleUpdateScript [
   &ParticleState(Kill Particles When Lifetime Has Elapsed=false, Let Infinitely Lived Particles Die When Emitter Deactivates=true)
   &DynamicMaterialParameters(
    Index 0 Param 2=1
    Index 0 Param 3=1
    Index 0 Param 4=1
    Module.Param0WriteEnabled=true
    Module.Param1WriteEnabled=true
    Module.Param2WriteEnabled=true
    Module.Param3WriteEnabled=true
   )
  ]
  rendererClasses [&NiagaraSpriteRendererProperties]
  renderers [{rendererClass=&NiagaraSpriteRendererProperties}]
 }
 Particle_Emitter {
  emitterUpdateScript [
   &EmitterState(Inactive Response=::NewEnumerator1, Life Cycle Mode=::NewEnumerator1, Loop Behavior=::NewEnumerator1, Loop Duration=1, UseLoopDelay=true)
   &SpawnBurst_Instantaneous(Use Spawn Probability=false, UseLoopCountLimit=false)
  ]
  particleSpawnScript [
   &InitializeParticle(Sprite Size=<300 850>, Sprite Size Mode=::NewEnumerator4, UsePositionOffset=false, Write Material Random=true)
   &Color(Color=&MakeLinearColorFromVectorAndFloat("Vector (RGB)"=<0 25 23.5576>, "Float (Alpha)"=1))
   &AddVectorToPosition(Vector:Vector3f=&TransformVector(Vector=<0 0 40>, Destination Space=::Simulation))
  ]
  particleUpdateScript [
   &ParticleState(Kill Particles When Lifetime Has Elapsed=false, Let Infinitely Lived Particles Die When Emitter Deactivates=true)
   &CascadeConversion_LightProperties(LightBrightnessScale=32, LightColorScale=<1 1 1>, LightRadiusScale=75)
  ]
  rendererClasses [&NiagaraLightRendererProperties]
  renderers [{rendererClass=&NiagaraLightRendererProperties, rendererIndex=0}]
 }
}
summary {
 systemName=NS_RaveNinjaFemale1H_Idle
 emitters [
  {emitterName=DecalsStart, rendererClasses [&NiagaraSpriteRendererProperties]}
  {emitterName=Particle_Emitter, rendererClasses [&NiagaraLightRendererProperties]}
  {emitterName=LowGlow, rendererClasses [&NiagaraSpriteRendererProperties]}
  {emitterName=LowGlow_LockX, rendererClasses [&NiagaraSpriteRendererProperties]}
 ]
 userVariables []
}
