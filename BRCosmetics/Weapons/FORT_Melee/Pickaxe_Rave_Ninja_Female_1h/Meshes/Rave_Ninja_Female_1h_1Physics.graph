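# Physics asset for the one-handed Rave Ninja Female pickaxe, defining bodies and joints for its dangling ribbons.
# Three two-segment chains of slim capsules each attach to the dyn_ribbon_mstr body through constraints with matching moderate swing and tight twist limits.
# The master body carries a tiny capsule at its bone plus a large offset capsule, which looks like a stand-in collision volume for the weapon.
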PhysicsAsset /BRCosmetics/Weapons/FORT_Melee/Pickaxe_Rave_Ninja_Female_1h/Meshes/Rave_Ninja_Female_1h_1Physics
bodies {
 dyn_ribbon_01 [{shapeType=Capsule, length=8.2845, radius=1.029, center=<4.071 -0.0431 -0.194>, rotation=<rot 0.063 91.0058 -89.61>}]
 dyn_ribbon_02 [{shapeType=Capsule, length=6.3793, radius=0.8108, center=<3.8696 0.8548 0.1988>, rotation=<rot -0.2571 90.799 -88.3959>}]
 dyn_ribbon_03 [{shapeType=Capsule, length=8.2845, radius=1.029, center=<4.071 -0.0431 -0.194>, rotation=<rot 0.063 91.0058 -89.61>}]
 dyn_ribbon_04 [{shapeType=Capsule, length=6.3793, radius=0.8108, center=<3.894 -0.8974 0.2067>, rotation=<rot -0.2571 90.799 -88.3959>}]
 dyn_ribbon_05 [{shapeType=Capsule, length=8.2845, radius=1.029, center=<4.071 -0.0431 -0.194>, rotation=<rot 0.063 91.0058 -89.61>}]
 dyn_ribbon_06 [{shapeType=Capsule, length=6.3793, radius=0.8108, center=<3.9093 0.1229 -0.8503>, rotation=<rot -0.2571 90.799 -88.3959>}]
 dyn_ribbon_mstr [
  {shapeType=Capsule, length=0.505, radius=0.505, rotation=<rot 0 -90 0>}
  {shapeType=Capsule, length=63.25, radius=15, center=<-28.539 -0.0001 0>, rotation=<rot 90 -0.0003 -0.0002>}
 ]
}
constraints [
 {bone1Name=dyn_ribbon_01, bone2Name=dyn_ribbon_mstr, swing1LimitDegrees=45, swing2LimitDegrees=45, twistLimitDegrees=5}
 {bone1Name=dyn_ribbon_02, bone2Name=dyn_ribbon_01, swing1LimitDegrees=45, swing2LimitDegrees=45, twistLimitDegrees=5}
 {bone1Name=dyn_ribbon_03, bone2Name=dyn_ribbon_mstr, swing1LimitDegrees=45, swing2LimitDegrees=45, twistLimitDegrees=5}
 {bone1Name=dyn_ribbon_04, bone2Name=dyn_ribbon_03, swing1LimitDegrees=45, swing2LimitDegrees=45, twistLimitDegrees=5}
 {bone1Name=dyn_ribbon_05, bone2Name=dyn_ribbon_mstr, swing1LimitDegrees=45, swing2LimitDegrees=45, twistLimitDegrees=5}
 {bone1Name=dyn_ribbon_06, bone2Name=dyn_ribbon_05, swing1LimitDegrees=45, swing2LimitDegrees=45, twistLimitDegrees=5}
]
